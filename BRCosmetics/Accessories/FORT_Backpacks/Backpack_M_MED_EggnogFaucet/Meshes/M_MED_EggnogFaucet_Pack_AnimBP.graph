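# Animation Blueprint for the EggnogFaucet backpack cosmetic, offsetting its base_backpack bone to follow the wearer's spine bend.
# Each update maps the owner mesh's spine_05 yaw, relative to spine_01 and reference pose, through clamped ranges and constant-alpha ease nodes into rotation and translation.
# An additive bone-space modifier applies these values.
# A second modifier adds a fixed X offset, blended by interpolated alpha when not skydiving.
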AnimBlueprint /BRCosmetics/Accessories/FORT_Backpacks/Backpack_M_MED_EggnogFaucet/Meshes/M_MED_EggnogFaucet_Pack_AnimBP
target_skeleton=&/M_MED_EggnogFaucet_Pack_Skeleton
graphs [
 AnimGraph {nodes [
   n1 = "Transform (Modify) Bone - Bone: base_backpack"(
    ComponentPose=Local To Component(LocalPose=Input Pose())
    Alpha=1
    Translation=Get BP Translate() -> (BP Translate:Vector)
    Rotation=Get BP Rotate() -> (BP Rotate:Rotator)
    Scale="1,1,1"
   ) {
    rotationMode=BMM_Additive
    rotationSpace=BCS_BoneSpace
    translationMode=BMM_Additive
    translationSpace=BCS_BoneSpace
    alphaBoolBlend {blendOption=Linear}
    alphaScaleBiasClamp {bInterpResult=true}
    boneToModify {boneName=base_backpack}
    rotation {roll=10}
   }
   n2 = "Transform (Modify) Bone - Bone: base_backpack"(ComponentPose=@n1, Alpha=Get isNotSkydiving() -> (isNotSkydiving:double), Translation="2.25,0,0", Rotation="0,0,0", Scale="1,1,1") {
    rotationMode=BMM_Additive
    rotationSpace=BCS_BoneSpace
    translationMode=BMM_Additive
    translationSpace=BCS_BoneSpace
    alphaBoolBlend {blendOption=Linear}
    alphaScaleBiasClamp {bInterpResult=true, interpSpeedIncreasing=5}
    boneToModify {boneName=base_backpack}
    translation {x=2.25}
   }
   Output Pose(Result=Component To Local(ComponentPose=@n2)) {layerGroup=None}
  ]}
 EventGraph {nodes [
   Event BlueprintUpdateAnimation()
   n1 = GetOwnerSkeletalMeshComponent()
   n2 = BreakRotator(InRot=BreakTransform(InTransform=GetDeltaTransformFromRefPose(self=@n1, BoneName=spine_05, BaseName=spine_01)).Rotation)
   n3 = FInterpEaseInOut!(A:double=0, B:double=MapRangeClamped(Value=@n2.Yaw, InRangeA=0, InRangeB=-20, OutRangeB=20), Alpha:double=0, Exponent:double=0) -> (ReturnValue:double)
   > Set BP Rotate(BP Rotate:Rotator=MakeRotator(Roll=@n3, Pitch=@n3)) -> (Output_Get:Rotator)
   n4 = MapRangeClamped(Value=@n2.Yaw, InRangeA=0, InRangeB=-10, OutRangeB=5.75)
   > Set BP Translate(BP Translate:Vector=MakeVector(X=FInterpEaseInOut(A=0, B=@n4, Alpha=0, Exponent=0), Y=0, Z=1.5)) -> (Output_Get:Vector)
   n5 = Get bIsSkyDiving()
   > Set isNotSkydiving(isNotSkydiving:double="To Float (Boolean)"(InBool=NOT Boolean(A=@n5))) -> (Output_Get:double)
  ]}
]
variables [
 BP Rotate {type=Rotator}
 BP Translate {type=Vector}
 isNotSkydiving {type=number}
]
